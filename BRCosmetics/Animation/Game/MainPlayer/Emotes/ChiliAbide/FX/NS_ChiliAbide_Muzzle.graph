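# Niagara muzzle flash effect for the ChiliAbide emote, built from two short-lived burst emitters.
# The streak emitter spawns a few randomly sized, fanned meshes that scale up over life with an orange color curve.
# The core flash fires two staggered bursts, tagging particles by spawn group for visibility, and drives three sprite renderers, including lens-flare sizing, plus a blue light.
# The exposed user parameters are unreferenced in the shown modules.
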NiagaraSystem /BRCosmetics/Animation/Game/MainPlayer/Emotes/ChiliAbide/FX/NS_ChiliAbide_Muzzle
dependencies {
 usedDataInterfaces [
  &NiagaraDataInterfaceCurve
  &NiagaraDataInterfaceVector2DCurve
  &NiagaraDataInterfaceVectorCurve
 ]
 usedDynamicInputs [&RandomRangeFloat]
 usedModules [&SystemState]
}
emitters {
 CoreMuzzleFlashStreak001 {
  emitterUpdateScript [
   &EmitterState()
   &SpawnBurst_Instantaneous(Spawn Count=&RandomRangeInt(Minimum=3, Maximum=5), Use Spawn Probability=false)
  ]
  particleSpawnScript [
   /Game/Effects/Niagara/Modules/Orientation/FN_Fan(Spin:NiagaraFloat=&UniformRangedFloat(Minimum=0, Maximum=1))
   &InitializeParticle(
    Lifetime Max=0.1
    Lifetime Min=0.05
    Lifetime Mode=::NewEnumerator1
    Mesh Scale Max=<0.3 0.2 0.2>
    Mesh Scale Min=<0.1 0.1 0.1>
    Mesh Scale Mode=::NewEnumerator2
    UsePositionOffset=false
    Write Material Random=true
   )
   /Niagara/Modules/Spawn/Initialization/InitializeParticle(
    Lifetime=&UniformRangedFloat(Minimum=0.1, Maximum=0.1)
    Mesh Scale=&UniformRangedVector(Minimum=<0.9 0.2 0.2>, Maximum=<0.9 0.2 0.2>)
    Module.Write Color=true
    Module.Write Lifetime=true
    Module.Write Mass=true
    Module.Write Material Random=true
    Module.Write Position=true
    Module.Write Scale=true
    Module.Write SpriteRotation=false
    Module.Write SpriteSize=true
    Position=&Add_Vector(A=/Niagara/DynamicInputs/Helpers/SimulationPosition(), B=<0 0 0>)
    Sprite Size=<100 100>
   ) {enabled=false, module=InitializeParticle001}
   &InitialMeshOrientation(
    Mesh Orientation Mode=::NewEnumerator7
    Rotation=&RandomRangeVector(Minimum=<0 -0.1 0>, Maximum=<0 0.1 0>)
    Use Rotation Vector=true
   )
  ]
  particleUpdateScript [
   &ParticleState()
   &Color(
    Color=<rgba 20 10.5088 0 1>
    Scale Color=&VectorFromCurve(VectorCurve=DataInterface() {propertyValues {XCurve {keys [<at 0.7 12>,<at 1 1>]}, YCurve {keys [<at 0.7 12>,<at 1 1>]}, ZCurve {keys [<at 0.7 12>,<at 1 1>]}}})
   )
   &ScaleMeshSize(Scale Factor=&VectorFromFloat(Value=&FloatFromCurve(
      FloatCurve=DataInterface() {propertyValues {Curve {keys [{arriveTangent=2.2457, interpMode=RCIM_Cubic, leaveTangent=2.2457, tangentMode=RCTM_User, time=0, value=0}, <at 1 1>]}}}
      Scale Curve=0.5
     )))
   &FN_Sprite_Attributes(
    Create SpriteSize Attribute=true
    SpriteSize=&Vector2DFromFloat(Value=50)
    Write to Module Namespace=true
   ) {enabled=false, module=FN_Sprite}
   /SetVariables(Particles.SpriteColor=&MultiplyLinearColor(Linear Color=$Particles.Color, Scale Factor=<1 1 1 0.2>)) {isSetParameters=true}
  ]
  rendererClasses [&NiagaraMeshRendererProperties]
  renderers [{rendererClass=&NiagaraMeshRendererProperties, rendererIndex=0}]
 }
 FN_Weapons_Core_Flash_BETA {
  emitterUpdateScript [
   &EmitterState()
   &SpawnBurst_Instantaneous(Spawn Count=1, Use Spawn Probability=false)
   &SpawnBurst_Instantaneous(Spawn Group=1, Spawn Time=0.03, Use Spawn Probability=false, UseLoopCountLimit=false) {module=SpawnBurst_Instantaneous001}
  ]
  particleSpawnScript [
   /Niagara/Modules/Spawn/Initialization/V2/InitializeParticle(
    Color=<rgba 1 0.5254 0 1>
    Lifetime=0.15
    Sprite Size=<150 30>
    Sprite Size Mode=::NewEnumerator4
    UsePositionOffset=false
    Write Material Random=true
   ) {module=InitializeParticle001}
   /SetVariables(Particles.Linear Color_LensFlare:LinearColor=<rgba 5 2.6272 0 0.5>, Particles.SpriteSize_LensFlare:Vector2f=<40 40>) {isSetParameters=true}
   /SetVariables(Particles.VisibilityTag=&SpawnGroup()) {isSetParameters=true}
   &CameraOffset(Camera Offset Amount=5)
  ]
  particleUpdateScript [
   &ParticleState()
   &ScaleSpriteSize(
    Non-Uniform Curve Scale=<2 0.5>
    Non-Uniform Curve Sprite Scale=DataInterface() {propertyValues {
      XCurve {keys [
        {arriveTangent=4.6568, interpMode=RCIM_Cubic, leaveTangent=4.6568, tangentMode=RCTM_User, time=0, value=0}
        <at 0.3 1 RCIM_Cubic>
        <at 1 0 RCIM_Cubic>
       ]}
      YCurve {keys [{tangentMode=RCTM_User, time=1, value=1}]}
     }}
    Scale Sprite Size Mode=::NewEnumerator3
   )
   &FN_Light_Attributes(
    Color=<rgba 0 0.25 1 1>
    Create Color Attribute=true
    Create Exponent Attribute=true
    Create Position Attribute=true
    Enable Particle Light Enable=true
    Exponent=4
    Light Enabled=true
    Position=<0 4.5 0>
    Scale Color=3
   )
   /SetVariables(Particles.SpriteSize_UpdateLensFlare:Vector2f=&Multiply_Vector2D(
     A=$Particles.Initial.SpriteSize_LensFlare
     B=&Vector2DFromCurve(
      Vector2Curve=DataInterface() {propertyValues {
        XCurve {keys [
          {arriveTangent=0.0671, interpMode=RCIM_Cubic, leaveTangent=0.0671, tangentMode=RCTM_User, time=0, value=1}
          {arriveTangent=1.2594, interpMode=RCIM_Cubic, leaveTangent=-1.1099, tangentMode=RCTM_Break, time=0.45, value=1.5}
          {arriveTangent=0.0671, interpMode=RCIM_Cubic, leaveTangent=0.0671, tangentMode=RCTM_User, time=1, value=1}
         ]}
        YCurve {keys [
          {arriveTangent=2.993, interpMode=RCIM_Cubic, leaveTangent=2.993, tangentMode=RCTM_User, time=0, value=0}
          {arriveTangent=-0.4996, interpMode=RCIM_Cubic, leaveTangent=-0.4996, tangentMode=RCTM_User, time=0.5, value=1}
          {arriveTangent=-0.0187, interpMode=RCIM_Cubic, leaveTangent=-0.0187, tangentMode=RCTM_User, time=1, value=0}
         ]}
       }}
      ScaleCurve=<1 1>
     )
    )) {isSetParameters=true}
   /SetVariables(Particles.SpriteSize_UpdateFlareVertical:Vector2f=&Multiply_Vector2D(
     A=$Particles.Initial.SpriteSize
     B=&Vector2DFromCurve(
      Vector2Curve=DataInterface() {propertyValues {
        XCurve {keys [<at 0 1>,<at 1 1>]}
        YCurve {
         defaultValue=0
         keys [
          {arriveTangent=4.6568, interpMode=RCIM_Cubic, leaveTangent=4.6568, tangentMode=RCTM_User, time=0, value=0}
          <at 0.3 1 RCIM_Cubic>
          <at 1 0 RCIM_Cubic>
         ]
        }
       }}
      ScaleCurve=<0.1 4>
     )
    )) {isSetParameters=true}
  ]
  rendererClasses [&NiagaraLightRendererProperties, &NiagaraSpriteRendererProperties]
  renderers [
   {rendererClass=&NiagaraLightRendererProperties, rendererIndex=0}
   {rendererClass=&NiagaraSpriteRendererProperties, rendererIndex=1}
   {rendererClass=&NiagaraSpriteRendererProperties, rendererIndex=2}
   {rendererClass=&NiagaraSpriteRendererProperties, rendererIndex=3}
  ]
 }
}
summary {
 systemName=NS_ChiliAbide_Muzzle
 emitters [
  {
   emitterName=FN_Weapons_Core_Flash_BETA
   rendererClasses [&NiagaraLightRendererProperties, &NiagaraSpriteRendererProperties]
  }
  {emitterName=CoreMuzzleFlashStreak001, rendererClasses [&NiagaraMeshRendererProperties]}
 ]
 userVariables [
  User.ChargeUpLength {type=NiagaraFloat, defaultValue=0, description=""}
  User.ColorThemeA {type=LinearColor, description="", defaultValue {struct=LinearColor, value=<rgba 1 1 1 1>}}
  User.ColorThemeB {type=LinearColor, description="", defaultValue {struct=LinearColor, value=<rgba 1 1 1 1>}}
  User.Skeletal Mesh {
   type=NiagaraDataInterfaceSkeletalMesh
   defaultValue {
    struct=NiagaraExt_VariableValue_DataInterface
    value {dataInterface=&self:NiagaraDataInterfaceSkeletalMesh_0}
   }
  }
 ]
}
