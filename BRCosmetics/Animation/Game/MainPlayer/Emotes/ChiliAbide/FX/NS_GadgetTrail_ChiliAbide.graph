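# Niagara trail effect for the ChiliAbide emote gadget, built from three sprite emitters that spawn continuously at sockets read from the gadget.
# Lines emits short-lived streaks with random backward velocity and drag, plus a front-end camera offset.
# Smear_Center (TieFighter_Exhaust socket) and Smear_Wings re-sample sockets every update, stretching and fading as they age.
# All three tint RGB through a color curve indexed by the NPC-supplied remapped time of day.
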NiagaraSystem /BRCosmetics/Animation/Game/MainPlayer/Emotes/ChiliAbide/FX/NS_GadgetTrail_ChiliAbide
dependencies {
 usedDataInterfaces [
  &NiagaraDataInterfaceCurve
  &NiagaraDataInterfaceVector2DCurve
  &NiagaraDataInterfaceColorCurve
  &NiagaraDataInterfaceSocketReader
 ]
 usedDynamicInputs [
  &MergeHelper_FloatSwitch
  &RandomRangeFloat
  &RandomRangeInt
  &RandomRangeLinearColor
 ]
 usedModules [
  &self:SystemSpawnScript.SystemScriptSource.SystemScriptGraph.NiagaraNodeAssignment_2.SetVariables_#1
  &self:SystemSpawnScript.SystemScriptSource.SystemScriptGraph.NiagaraNodeAssignment_1.SetVariables_#2
  &self:SystemSpawnScript.SystemScriptSource.SystemScriptGraph.NiagaraNodeAssignment_0.SetVariables_#3
  &SystemState
 ]
}
emitters {
 Lines {
  emitterUpdateScript [
   &EmitterState(Life Cycle Mode=::NewEnumerator1, Loop Duration=0.95, UseLoopDelay=false)
   &SpawnRate(SpawnRate=100, Use Spawn Probability=false) {module=SpawnRate001}
  ]
  particleSpawnScript [
   &InitializeParticle(
    Color Maximum=<rgba 0.03 0.03 0.03 8>
    Color Minimum=<rgba 0.1 0.1 0.1 10>
    Color Mode=::NewEnumerator2
    Lifetime Max=0.08
    Lifetime Min=0.025
    Lifetime Mode=::NewEnumerator1
    Position Mode=::NewEnumerator1
    Sprite Size Max=<20 50>
    Sprite Size Min=<2 10>
    Sprite Size Mode=::NewEnumerator2
    Sprite UV Mode=::NewEnumerator3
    UsePositionOffset=false
    Write Material Random=true
   )
   &AddVelocity(Velocity=&RandomRangeVector(Minimum=<0 -500 0>, Maximum=<1 -50 0>)) {module=AddVelocity001}
   &CameraOffset(Camera Offset Amount=&FN_FrontEndFloatModifier(Gameplay=0, Front End=-30))
   &SocketLocation(
    "Inherit Velocity (Scale)"=0
    OverrideRandomSeed=false
    SamplePosOffset=true
    Sampled Orientation Coordinate Space:ENiagaraCoordinateSpace=::Local
    Sampled Position Offset=&RandomRangeVector(Minimum=<0 5 -10>, Maximum=<0 5 10>)
    Socket Reader=$System.Socket Reader_Gadget
    Socket Sampling Mode=::NewEnumerator1
    VelocitySpeedLimit=false
   )
  ]
  particleUpdateScript [
   &ParticleState()
   &Drag(Drag=5, Use Linear Drag=true, Use Rotational Drag=false)
   &SolveForcesAndVelocity(Clamp Velocity=false, Limit Acceleration=false)
   &ScaleColor(
    Scale RGB=&MakeVectorFromLinearColorRGB(Color=&ColorFromCurve(
      ColorCurve=DataInterface() {propertyValues {
        AlphaCurve {keys [<at 0 1 RCIM_Cubic>]}
        BlueCurve {keys [
          <at 0.0513 1.15>
          <at 0.1437 0.5>
          {arriveTangent=1.2068, interpMode=RCIM_Cubic, leaveTangent=1.2068, time=0.3527, value=1}
          {arriveTangent=-0.3926, interpMode=RCIM_Cubic, leaveTangent=-0.3926, time=0.558, value=1}
          <at 0.6703 0.8753>
          {arriveTangent=1.0578, interpMode=RCIM_Cubic, leaveTangent=1.0578, time=0.8924, value=0.8}
          <at 0.93 1.15>
         ]}
        GreenCurve {keys [
          <at 0.0513 0.5751>
          <at 0.1437 0.5417>
          {arriveTangent=1.1063, interpMode=RCIM_Cubic, leaveTangent=1.1063, time=0.3527, value=1}
          {arriveTangent=-0.1515, interpMode=RCIM_Cubic, leaveTangent=-0.1515, time=0.558, value=1}
          <at 0.6703 0.9519>
          {arriveTangent=-1.4508, interpMode=RCIM_Cubic, leaveTangent=-1.4508, time=0.8924, value=0.8167}
          <at 0.93 0.5751>
         ]}
        RedCurve {keys [
          <at 0.0513 0.2877>
          <at 0.1437 1>
          <at 0.3527 1 RCIM_Cubic>
          {arriveTangent=0.2119, interpMode=RCIM_Cubic, leaveTangent=0.2119, time=0.558, value=1}
          <at 0.6703 1.0673>
          {arriveTangent=-3.002, interpMode=RCIM_Cubic, leaveTangent=-3.002, time=0.8924, value=1}
          <at 0.93 0.2877>
         ]}
       }}
      CurveIndex=Linked() {linkedVariable {name=NPC.FortniteNPC.RemappedTimeOfDay_0-1, type=NiagaraFloat}}
     ))
    ScaleA=true
    ScaleRGB=true
   )
  ]
  rendererClasses [&NiagaraSpriteRendererProperties]
  renderers [{rendererClass=&NiagaraSpriteRendererProperties}]
 }
 Smear_Center {
  emitterUpdateScript [
   &EmitterState(Life Cycle Mode=::NewEnumerator1, Loop Duration=1, UseLoopDelay=false)
   &SpawnRate(SpawnRate=50, Use Spawn Probability=false)
  ]
  particleSpawnScript [
   &InitializeParticle(
    Color Mode=::NewEnumerator0
    Lifetime=0.08
    Sprite Size Max=<180 15>
    Sprite Size Min=<40 5>
    Sprite Size Mode=::NewEnumerator2
    Sprite UV Mode=::NewEnumerator2
    UsePositionOffset=false
    Write Material Random=true
   )
   &SpriteFacingAndAlignment(Alignment Coordinate Space=::Local, Facing Coordinate Space=::Local, Sprite Alignment=<0 0 1>, WriteAlignment=true, WriteFacing=true)
  ]
  particleUpdateScript [
   &ParticleState()
   &SocketLocation(
    "Inherit Velocity (Scale)"=0
    OverrideRandomSeed=false
    SamplePosOffset=true
    Sampled Orientation Coordinate Space:ENiagaraCoordinateSpace=::Local
    Sampled Position Offset=&RandomRangeVector(Minimum=<-2 0 -2>, Maximum=<2 0 2>)
    Socket Reader=DataInterface() {propertyValues {
      AttachComponentClass=null
      AttachComponentTag=None
      EditorPreviewAsset=/BRCosmetics/Animation/Props/Gadgets_ChiliAbide/Meshes/ChiliAbide_Gadget
      SourceActorRef=null
      SourceAsset=null
      SourceMode=Default
      bRequireCurrentFrameData=true
      bUpdateSocketsPerFrame=true
      FilteredSockets [TieFighter_Exhaust]
      ObjectParameterBinding {parameter {name=None, typeDefHandle {registeredTypeIndex=1}, varData []}}
     }}
    Socket Sampling Mode=::NewEnumerator1
    VelocitySpeedLimit=false
   )
   &ScaleSpriteSize(
    Non-Uniform Curve Scale=<1.1 1>
    Non-Uniform Curve Sprite Scale=DataInterface() {propertyValues {
      XCurve {keys [{tangentMode=RCTM_User, time=0, value=0}, {tangentMode=RCTM_User, time=1, value=1}]}
      YCurve {keys [{tangentMode=RCTM_User, time=0, value=1}, {tangentMode=RCTM_User, time=1, value=1}]}
     }}
    Scale Sprite Size Mode=::NewEnumerator3
   )
   &ScaleColor(
    Scale Alpha=&FloatFromCurve(FloatCurve=DataInterface() {propertyValues {Curve {keys [<at 0.3 1 RCIM_Cubic>, {arriveTangent=-3.9988, interpMode=RCIM_Cubic, leaveTangent=-3.9988, tangentMode=RCTM_User, time=1, value=0}]}}})
    Scale RGB=&MakeVectorFromLinearColorRGB(Color=&ColorFromCurve(
      ColorCurve=DataInterface() {propertyValues {
        AlphaCurve {keys [<at 0 1 RCIM_Cubic>]}
        BlueCurve {keys [
          <at 0.0513 1.15>
          <at 0.1437 0.5>
          {arriveTangent=1.2068, interpMode=RCIM_Cubic, leaveTangent=1.2068, time=0.3527, value=1}
          {arriveTangent=-0.3926, interpMode=RCIM_Cubic, leaveTangent=-0.3926, time=0.558, value=1}
          <at 0.6703 0.8753>
          {arriveTangent=1.0578, interpMode=RCIM_Cubic, leaveTangent=1.0578, time=0.8924, value=0.8}
          <at 0.93 1.15>
         ]}
        GreenCurve {keys [
          <at 0.0513 0.5751>
          <at 0.1437 0.5417>
          {arriveTangent=1.1063, interpMode=RCIM_Cubic, leaveTangent=1.1063, time=0.3527, value=1}
          {arriveTangent=-0.1515, interpMode=RCIM_Cubic, leaveTangent=-0.1515, time=0.558, value=1}
          <at 0.6703 0.9519>
          {arriveTangent=-1.4508, interpMode=RCIM_Cubic, leaveTangent=-1.4508, time=0.8924, value=0.8167}
          <at 0.93 0.5751>
         ]}
        RedCurve {keys [
          <at 0.0513 0.2877>
          <at 0.1437 1>
          <at 0.3527 1 RCIM_Cubic>
          {arriveTangent=0.2119, interpMode=RCIM_Cubic, leaveTangent=0.2119, time=0.558, value=1}
          <at 0.6703 1.0673>
          {arriveTangent=-3.002, interpMode=RCIM_Cubic, leaveTangent=-3.002, time=0.8924, value=1}
          <at 0.93 0.2877>
         ]}
       }}
      CurveIndex=Linked() {linkedVariable {name=NPC.FortniteNPC.RemappedTimeOfDay_0-1, type=NiagaraFloat}}
     ))
    ScaleA=true
    ScaleRGB=true
   )
  ]
  rendererClasses [&NiagaraSpriteRendererProperties]
  renderers [{rendererClass=&NiagaraSpriteRendererProperties}]
 }
 Smear_Wings {
  emitterUpdateScript [
   &EmitterState(Life Cycle Mode=::NewEnumerator1, Loop Duration=1, UseLoopDelay=false)
   &SpawnRate(SpawnRate=100, Use Spawn Probability=false)
  ]
  particleSpawnScript [
   &InitializeParticle(
    Color Mode=::NewEnumerator0
    Lifetime=0.08
    Sprite Size Max=<150 30>
    Sprite Size Min=<20 5>
    Sprite Size Mode=::NewEnumerator2
    Sprite UV Mode=::NewEnumerator2
    UsePositionOffset=false
    Write Material Random=true
   )
   &SpriteFacingAndAlignment(Alignment Coordinate Space=::Local, Facing Coordinate Space=::Local, Sprite Alignment=<0 0 1>, WriteAlignment=true, WriteFacing=true)
  ]
  particleUpdateScript [
   &ParticleState()
   &SocketLocation(
    "Inherit Velocity (Scale)"=0
    OverrideRandomSeed=false
    SamplePosOffset=false
    Sampled Orientation Coordinate Space:ENiagaraCoordinateSpace=::Local
    Sampled Position Offset=<0 5 -9>
    Socket Reader=$System.Socket Reader_Gadget
    Socket Sampling Mode=::NewEnumerator1
    VelocitySpeedLimit=false
   )
   &ScaleSpriteSize(
    Non-Uniform Curve Scale=<1.1 1>
    Non-Uniform Curve Sprite Scale=DataInterface() {propertyValues {
      XCurve {keys [{tangentMode=RCTM_User, time=0, value=0}, {tangentMode=RCTM_User, time=1, value=1}]}
      YCurve {keys [{tangentMode=RCTM_User, time=0, value=1}, {tangentMode=RCTM_User, time=1, value=1}]}
     }}
    Scale Sprite Size Mode=::NewEnumerator3
   )
   &ScaleColor(
    Scale Alpha=&FloatFromCurve(FloatCurve=DataInterface() {propertyValues {Curve {keys [<at 0.3 1 RCIM_Cubic>, {arriveTangent=-3.9988, interpMode=RCIM_Cubic, leaveTangent=-3.9988, tangentMode=RCTM_User, time=1, value=0}]}}})
    Scale RGB=&MakeVectorFromLinearColorRGB(Color=&ColorFromCurve(
      ColorCurve=DataInterface() {propertyValues {
        AlphaCurve {keys [<at 0 1 RCIM_Cubic>]}
        BlueCurve {keys [
          <at 0.0513 1.15>
          <at 0.1437 0.5>
          {arriveTangent=1.2068, interpMode=RCIM_Cubic, leaveTangent=1.2068, time=0.3527, value=1}
          {arriveTangent=-0.3926, interpMode=RCIM_Cubic, leaveTangent=-0.3926, time=0.558, value=1}
          <at 0.6703 0.8753>
          {arriveTangent=1.0578, interpMode=RCIM_Cubic, leaveTangent=1.0578, time=0.8924, value=0.8}
          <at 0.93 1.15>
         ]}
        GreenCurve {keys [
          <at 0.0513 0.5751>
          <at 0.1437 0.5417>
          {arriveTangent=1.1063, interpMode=RCIM_Cubic, leaveTangent=1.1063, time=0.3527, value=1}
          {arriveTangent=-0.1515, interpMode=RCIM_Cubic, leaveTangent=-0.1515, time=0.558, value=1}
          <at 0.6703 0.9519>
          {arriveTangent=-1.4508, interpMode=RCIM_Cubic, leaveTangent=-1.4508, time=0.8924, value=0.8167}
          <at 0.93 0.5751>
         ]}
        RedCurve {keys [
          <at 0.0513 0.2877>
          <at 0.1437 1>
          <at 0.3527 1 RCIM_Cubic>
          {arriveTangent=0.2119, interpMode=RCIM_Cubic, leaveTangent=0.2119, time=0.558, value=1}
          <at 0.6703 1.0673>
          {arriveTangent=-3.002, interpMode=RCIM_Cubic, leaveTangent=-3.002, time=0.8924, value=1}
          <at 0.93 0.2877>
         ]}
       }}
      CurveIndex=Linked() {linkedVariable {name=NPC.FortniteNPC.RemappedTimeOfDay_0-1, type=NiagaraFloat}}
     ))
    ScaleA=true
    ScaleRGB=true
   )
  ]
  rendererClasses [&NiagaraSpriteRendererProperties]
  renderers [{rendererClass=&NiagaraSpriteRendererProperties}]
 }
}
summary {
 systemName=NS_GadgetTrail_ChiliAbide
 emitters [
  {emitterName=Smear_Wings, rendererClasses [&NiagaraSpriteRendererProperties]}
  {emitterName=Lines, rendererClasses [&NiagaraSpriteRendererProperties]}
  {emitterName=Smear_Center, rendererClasses [&NiagaraSpriteRendererProperties]}
 ]
 userVariables [
  User.bIsFrontEnd {type=NiagaraBool}
  User.bIsFrontEndPreview {type=NiagaraBool}
  User.RibbonIDOffset {type=NiagaraInt32}
  User.Trail_Life {type=NiagaraFloat, defaultValue=0, description=""}
 ]
}
