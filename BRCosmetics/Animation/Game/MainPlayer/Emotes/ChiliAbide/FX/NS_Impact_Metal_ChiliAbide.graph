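# One-shot Niagara metal impact effect for the ChiliAbide emote: a brief orange flash with sparks.
# Flash001 bursts one mesh-and-sprite flash, FlashStreaks adds randomly fanned mesh streaks, and Spikes throws spark sprites that fall under gravity, shrink and darken toward red.
# A separate short-lived Light emitter pulses warm light, dimming with camera distance.
# Distance-driven blends enlarge the flash sprite, boost its alpha and widen the light radius farther away.
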NiagaraSystem /BRCosmetics/Animation/Game/MainPlayer/Emotes/ChiliAbide/FX/NS_Impact_Metal_ChiliAbide
dependencies {
 usedDataInterfaces [
  &NiagaraDataInterfaceCurve
  &NiagaraDataInterfaceVector2DCurve
  &NiagaraDataInterfaceColorCurve
  &NiagaraDataInterfaceVectorCurve
 ]
 usedDynamicInputs []
 usedModules [&SystemState]
}
emitters {
 Flash001 {
  emitterUpdateScript [
   &EmitterState(
    Enable Distance Culling=true
    Max Distance Response=::KillAfterParticlesFinish
    MaxDistance=30000
    Module.bUseMaxDistance=true
    Module.bUseMinDistance=false
    Reset Age On Awaken=true
    Scalability Mode=::NewEnumerator1
   )
   &SpawnBurst_Instantaneous(Spawn Count=1, Use Spawn Probability=false)
  ]
  particleSpawnScript [
   &InitializeParticle(
    Lifetime=&MaskFloatBySpawnGroup(Masked Float=0.1, Passthrough Float=&RandomRangeFloat(Minimum=0.5, Maximum=0.75))
    Mesh Scale=&UniformRangedVector(Minimum=<0.35 0.1 0.1>, Maximum=<0.45 0.2 0.2>)
    Mesh Scale Mode=::NewEnumerator4
    Sprite Size Mode=::NewEnumerator3
    Uniform Sprite Size=60
    UsePositionOffset=false
    Write Material Random=true
   ) {module=InitializeParticle001}
   /Game/Effects/Niagara/Modules/Orientation/FN_Fan(Spin:NiagaraFloat=&UniformRangedFloat(Minimum=0, Maximum=1))
   &CameraOffset(Camera Offset Amount=10)
   /SetVariables(Particles.LightPositionNew:Vector3f=&Add_Vector(
     A=Linked() {linkedVariable {name=Particles.Position, type=NiagaraPosition}}
     B=/Niagara/DynamicInputs/Transforms/TransformPosition(Position=<25 0 0>, Source Space=::Local, Destination Space=::Simulation)
    )) {isSetParameters=true, module=SetVariables_#1}
  ]
  particleUpdateScript [
   /SetVariables(Particles.DistanceBlend:NiagaraFloat=/Game/Effects/Niagara/DynamicInputs/Math/FN_Fit(
     Clamp:NiagaraBool=true
     Value:NiagaraFloat=$Emitter.Scalability.DistanceFraction
     SourceMin:NiagaraFloat=0.1
     SourceMax:NiagaraFloat=1
     DestMin:NiagaraFloat=0
     DestMax:NiagaraFloat=1
    )) {isSetParameters=true}
   &ParticleState()
   &Color(
    Color=<rgba 1 0.5 0 1>
    Scale Color=&VectorFromCurve(VectorCurve=DataInterface() {propertyValues {XCurve {keys [<at 0.7 12>,<at 1 1>]}, YCurve {keys [<at 0.7 12>,<at 1 1>]}, ZCurve {keys [<at 0.7 12>,<at 1 1>]}}})
   )
   &ScaleMeshSize(Scale Factor=&VectorFromFloat(Value=&FloatFromCurve(
      FloatCurve=DataInterface() {propertyValues {Curve {keys [{arriveTangent=2.2457, interpMode=RCIM_Cubic, leaveTangent=2.2457, tangentMode=RCTM_User, time=0, value=0}, <at 1 1>]}}}
      Scale Curve=&Lerp_Float(A=1, B=1, Alpha=$Particles.DistanceBlend)
     )))
   &FN_Sprite_Attributes(
    Create SpriteSize Attribute=true
    SpriteSize=&Vector2DFromFloat(Value=&Lerp_Float(A=100, B=350, Alpha=$Particles.DistanceBlend))
    Write to Module Namespace=true
   ) {module=FN_Sprite}
   /SetVariables(Particles.SpriteColor=&MultiplyLinearColor(
     Linear Color=$Particles.Color
     Scale Factor=&MakeVector4(X=1, Y=1, Z=1, W=&Lerp_Float(A=0.25, B=1.5, Alpha=$Particles.DistanceBlend))
    )) {isSetParameters=true, module=SetVariables_#2}
   &DynamicMaterialParameters(Write Parameter Index 0=false)
   &FN_Light_Attributes(
    Color=<rgba 0.2708 0.125 1 1>
    Create Color Attribute=true
    Create Exponent Attribute=true
    Enable Particle Light Enable=true
    Exponent=4
    Light Enabled=true
    Scale Color=5
   )
   &FN_Radial_Material(
    Base Sprite Offset X:NiagaraFloat=&RandomRangeFloat(Minimum=-1, Maximum=1)
    Base Sprite Offset Y:NiagaraFloat=&RandomRangeFloat(Minimum=-1, Maximum=1)
    Sprite Extrusion Speed=0.5
    Write to Module Namespace=true
   )
   &FN_Sprite_Attributes(
    Color=&MultiplyLinearColor(
     Linear Color=&LerpLinearColors(
      StartColor=<rgba 0.15 0.15 0.15 1>
      EndColor=<rgba 0 1.25 5 1>
      LerpFactor=&FloatFromCurve(FloatCurve=DataInterface() {propertyValues {Curve {keys [{interpMode=RCIM_Cubic, tangentMode=RCTM_User, time=0.2, value=1}, {interpMode=RCIM_Cubic, tangentMode=RCTM_User, time=0.4, value=0}]}}})
     )
     Scale Factor=&MakeVector4FromVectorAndFloat(
      "Vector (RGB)"=<1 1 1>
      "Float (Alpha)"=&FloatFromCurve(FloatCurve=DataInterface() {propertyValues {Curve {keys [
           {arriveTangent=9.0725, interpMode=RCIM_Cubic, leaveTangent=9.0725, tangentMode=RCTM_User, time=0, value=0}
           <at 0.1408 0.9847 RCIM_Cubic>
           {arriveTangent=-0.0405, interpMode=RCIM_Cubic, leaveTangent=-0.0405, tangentMode=RCTM_User, time=1, value=0}
          ]}}})
     )
    )
    Create Color Attribute=true
    Create Position Attribute=true
    Create SpriteSize Attribute=true
    Position:NiagaraPosition=&Lerp_Vector(
     A=<0 0 0>
     B=<15 1 1>
     Alpha=&FloatFromCurve(FloatCurve=DataInterface() {propertyValues {Curve {keys [
          {arriveTangent=2.0477, interpMode=RCIM_Cubic, leaveTangent=2.0477, tangentMode=RCTM_User, time=0, value=0}
          {interpMode=RCIM_Cubic, tangentMode=RCTM_User, time=1, value=1}
         ]}}})
    )
    SpriteSize=&Multiply_Vector2D_ByFloat(
     Float=&FloatFromCurve(
      FloatCurve=DataInterface() {propertyValues {Curve {keys [
          {arriveTangent=1.2221, interpMode=RCIM_Cubic, leaveTangent=1.2221, tangentMode=RCTM_User, time=0, value=0.25}
          {interpMode=RCIM_Cubic, leaveTangent=0.3375, time=1, value=1}
         ]}}}
      Scale Curve=1.5
     )
     Vector2D=&Vector2DFromFloat(Value=&RandomRangeFloat(Minimum=30, Maximum=40))
    )
    Write to Module Namespace=true
   )
   /SetVariables(
    Particles.Color_Glow:LinearColor=<rgba 5 3.3333 0 0.075>
    Particles.Size_Glow:Vector2f=&Multiply_Vector2D_ByFloat(Float=1.25, Vector2D=$Particles.Initial.SpriteSize)
   ) {isSetParameters=true}
  ]
  rendererClasses [&NiagaraMeshRendererProperties, &NiagaraSpriteRendererProperties]
  renderers [
   {rendererClass=&NiagaraMeshRendererProperties, rendererIndex=0}
   {rendererClass=&NiagaraSpriteRendererProperties, rendererIndex=1}
   {rendererClass=&NiagaraSpriteRendererProperties, rendererIndex=2}
  ]
 }
 FlashStreaks {
  emitterUpdateScript [
   &EmitterState()
   &SpawnBurst_Instantaneous(Spawn Count=&RandomRangeInt(Minimum=3, Maximum=5), Use Spawn Probability=false)
  ]
  particleSpawnScript [
   /Niagara/Modules/Spawn/Initialization/InitializeParticle(
    Lifetime=&UniformRangedFloat(Minimum=0.1, Maximum=0.1)
    Mesh Scale=&UniformRangedVector(Minimum=<0.05 0.04 0.04>, Maximum=<0.5 0.1 0.1>)
    Module.Write Color=true
    Module.Write Lifetime=true
    Module.Write Mass=true
    Module.Write Material Random=true
    Module.Write Position=true
    Module.Write Scale=true
    Module.Write SpriteRotation=false
    Module.Write SpriteSize=true
    Position=&Add_Vector(A=/Niagara/DynamicInputs/Helpers/SimulationPosition(), B=<0 0 0>)
    Sprite Size=<100 100>
   )
   &FN_Fan(Cone Angle=180, Spin:NiagaraFloat=&UniformRangedFloat(Minimum=0, Maximum=1))
   &CameraOffset(Camera Offset Amount=5)
   /SetVariables(Particles.LightPositionNew:Vector3f=&Add_Vector(
     A=Linked() {linkedVariable {name=Particles.Position, type=NiagaraPosition}}
     B=&TransformPosition(Position=<25 0 0>, Source Space=::Local, Destination Space=::Simulation)
    )) {isSetParameters=true, module=SetVariables_#1}
  ]
  particleUpdateScript [
   &ParticleState()
   &Color(
    Color=<rgba 1 0.5 0 1>
    Scale Color=&VectorFromCurve(VectorCurve=DataInterface() {propertyValues {XCurve {keys [<at 0.7 12>,<at 1 1>]}, YCurve {keys [<at 0.7 12>,<at 1 1>]}, ZCurve {keys [<at 0.7 12>,<at 1 1>]}}})
   )
   &ScaleMeshSize(Scale Factor=&VectorFromFloat(Value=&FloatFromCurve(FloatCurve=DataInterface() {propertyValues {Curve {keys [{arriveTangent=2.2457, interpMode=RCIM_Cubic, leaveTangent=2.2457, tangentMode=RCTM_User, time=0, value=0}, <at 1 1>]}}})))
   /SetVariables(Particles.SpriteColor=&MultiplyLinearColor(Linear Color=$Particles.Color, Scale Factor=<1 1 1 0.2>)) {isSetParameters=true, module=SetVariables_#2}
   &DynamicMaterialParameters(Write Parameter Index 0=false)
  ]
  rendererClasses [&NiagaraMeshRendererProperties]
  renderers [{rendererClass=&NiagaraMeshRendererProperties, rendererIndex=0}]
 }
 Light {
  emitterUpdateScript [
   &EmitterState(Enable Distance Culling=true, MaxDistance=4000, Module.bUseMaxDistance=true, Module.bUseMinDistance=true, Scalability Mode=::NewEnumerator1)
   &SpawnBurst_Instantaneous(Spawn Count=1, Use Spawn Probability=false)
  ]
  particleSpawnScript [
   /Niagara/Modules/Spawn/Initialization/V2/InitializeParticle(Lifetime=0.17, UsePositionOffset=false, Write Material Random=true) {module=InitializeParticle001}
   /SetVariables(Particles.LightPositionNew:Vector3f=&Add_Vector(
     A=Linked() {linkedVariable {name=Particles.Position, type=NiagaraPosition}}
     B=&TransformPosition(Position=<25 0 0>, Source Space=::Local, Destination Space=::Simulation)
    )) {isSetParameters=true, module=SetVariables_#1}
   /SetVariables(Particles.LightExponent=3) {isSetParameters=true}
  ]
  particleUpdateScript [
   /SetVariables(Particles.Light_Color:LinearColor=&MultiplyLinearColor(
     Linear Color=<rgba 1 0.3184 0.1481 1>
     Scale Factor=&MakeVector4_FromFloat(Float=&Multiply_Float(
       A=20
       B=&FloatFromCurve(
        FloatCurve=DataInterface() {propertyValues {Curve {keys [
            {interpMode=RCIM_Cubic, tangentMode=RCTM_User, time=0, value=0}
            {interpMode=RCIM_Cubic, tangentMode=RCTM_User, time=0.2, value=1}
            {tangentMode=RCTM_User, time=1, value=0}
           ]}}}
        Scale Curve=/Niagara/DynamicInputs/Scalability/ScalabilityDistanceBasedFloat(
         Curve:NiagaraDataInterfaceCurve=DataInterface() {propertyValues {Curve {keys [
             {interpMode=RCIM_Cubic, tangentMode=RCTM_Break, time=0, value=1}
             {arriveTangent=0.0625, interpMode=RCIM_Cubic, tangentMode=RCTM_Break, time=1, value=0}
            ]}}}
         CurveIndex:NiagaraFloat=$Emitter.Scalability.DistanceFraction
         Bias:NiagaraFloat=0
         Scale:NiagaraFloat=1
        )
       )
      ))
    )) {isSetParameters=true}
   /SetVariables(Particles.LightRadius=&Multiply_Float(
     A=40
     B=&ClampFloat(
      Float=&FN_LOD_Distance_Remap(Start Distance=0, End Distance=8000, Minimum Output=1, Maximum Output=2)
      Min=1
      Max=2
     )
    )) {isSetParameters=true}
   &ParticleState()
  ]
  rendererClasses [&NiagaraLightRendererProperties]
  renderers [{rendererClass=&NiagaraLightRendererProperties, rendererIndex=0}]
 }
 Spikes {
  emitterUpdateScript [
   &EmitterState()
   &SpawnBurst_Instantaneous(Spawn Count=&RandomRangeInt(Minimum=9, Maximum=12), Use Spawn Probability=false)
  ]
  particleSpawnScript [
   /Niagara/Modules/Spawn/Initialization/InitializeParticle(
    Color=<rgba 20 12.3188 1.5652 1>
    Lifetime=&UniformRangedFloat(Minimum=0.1, Maximum=0.3)
    Module.Write Color=true
    Module.Write Lifetime=true
    Module.Write Mass=true
    Module.Write Material Random=true
    Module.Write Position=true
    Module.Write Scale=false
    Module.Write SpriteRotation=false
    Module.Write SpriteSize=true
    Sprite Size=&RandomRangeVector2D(Minimum=<0.75 4>, Maximum=<1.5 8>)
   )
   &AddVelocityInCone(
    Cone Angle=&RandomRangeFloat(Minimum=110, Maximum=180)
    Module.Use Velocity Falloff On Cone Axis=false
    Velocity Strength=&Multiply_Float(
     A=&RandomRangeFloat(Minimum=100, Maximum=500)
     B=&ClampFloat(
      Float=&FN_LOD_Distance_Remap(Start Distance=0, End Distance=8000, Minimum Output=1, Maximum Output=1)
      Min=1
     )
    )
   )
  ]
  particleUpdateScript [
   &Drag(Drag=0.2, Use Linear Drag=true, Use Rotational Drag=false)
   &AccelerationForce(Acceleration=<0 0 -1200>)
   &SolveForcesAndVelocity(Clamp Velocity=false, Limit Acceleration=false)
   &ParticleState()
   &ScaleSpriteSize(
    Non-Uniform Curve Scale=<1 2>
    Non-Uniform Curve Sprite Scale=DataInterface() {propertyValues {
      XCurve {keys [{tangentMode=RCTM_User, time=0, value=1}, {tangentMode=RCTM_User, time=1, value=0.5}]}
      YCurve {keys [{tangentMode=RCTM_User, time=0, value=1}, {tangentMode=RCTM_User, time=1, value=0}]}
     }}
    Scale Sprite Size Mode=::NewEnumerator3
   )
   &ScaleColor(
    Scale RGB=&MakeVectorFromLinearColorRGB(Color=&ColorFromCurve(ColorCurve=DataInterface() {propertyValues {
        AlphaCurve {keys []}
        BlueCurve {keys [
          {arriveTangent=-0.8795, interpMode=RCIM_Cubic, leaveTangent=-0.8795, time=0, value=1}
          {arriveTangent=-0.8795, interpMode=RCIM_Cubic, leaveTangent=-0.8795, time=1, value=0.1205}
         ]}
        GreenCurve {keys [
          {arriveTangent=-0.974, interpMode=RCIM_Cubic, leaveTangent=-0.974, time=0, value=1}
          {arriveTangent=-0.974, interpMode=RCIM_Cubic, leaveTangent=-0.974, time=1, value=0.026}
         ]}
        RedCurve {keys [
          {arriveTangent=-0.6667, interpMode=RCIM_Cubic, leaveTangent=-0.6667, time=0, value=1}
          {arriveTangent=-0.6667, interpMode=RCIM_Cubic, leaveTangent=-0.6667, time=1, value=0.3333}
         ]}
       }}))
    ScaleA=true
    ScaleRGB=true
   )
  ]
  rendererClasses [&NiagaraSpriteRendererProperties]
  renderers [{rendererClass=&NiagaraSpriteRendererProperties}]
 }
}
summary {
 systemName=NS_Impact_Metal_ChiliAbide
 emitters [
  {emitterName=Spikes, rendererClasses [&NiagaraSpriteRendererProperties]}
  {emitterName=Light, rendererClasses [&NiagaraLightRendererProperties]}
  {emitterName=Flash001, rendererClasses [&NiagaraMeshRendererProperties, &NiagaraSpriteRendererProperties]}
  {emitterName=FlashStreaks, rendererClasses [&NiagaraMeshRendererProperties]}
 ]
 userVariables []
}
