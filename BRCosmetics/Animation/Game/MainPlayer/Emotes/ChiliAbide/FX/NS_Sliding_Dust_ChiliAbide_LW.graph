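# Niagara system for the ChiliAbide emote's sliding dust, spawned from skeletal mesh bones.
# Dust001 emits sprite puffs in left, right and middle spawn groups, while Foot_Spikes001 adds short-lived left/right meshes that follow their bones.
# User floats and IsSpawnDust gate spawning per group; system-level SildingColor tints both emitters.
# System-age curves time velocity and opacity changes, with separate front-end/gameplay values, indoor dimming and time-of-day spike tinting.
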NiagaraSystem /BRCosmetics/Animation/Game/MainPlayer/Emotes/ChiliAbide/FX/NS_Sliding_Dust_ChiliAbide_LW
dependencies {
 usedDataInterfaces [&NiagaraDataInterfaceCurve, &NiagaraDataInterfaceColorCurve, &NiagaraDataInterfaceSkeletalMesh]
 usedDynamicInputs [&self:ScratchDynamicInput]
 usedModules [
  &self:SystemSpawnScript.SystemScriptSource.SystemScriptGraph.NiagaraNodeAssignment_1.SetVariables_#1
  &SystemState
  &self:SetPeeloutFXArrayIndex
  &self:SystemSpawnScript.SystemScriptSource.SystemScriptGraph.NiagaraNodeAssignment_2.SetVariables_#2
 ]
}
emitters {
 Dust001 {
  emitterUpdateScript [
   &EmitterState()
   &SpawnRate(
    Spawn Probability=&MakeCustomFloatFromBool(Bool=$User.IsSpawnDust, True Float=$User.BurstIndexL, False Float=0)
    SpawnRate=35
    Use Spawn Probability=true
   )
   &SpawnRate(
    Spawn Probability=&MakeCustomFloatFromBool(Bool=$User.IsSpawnDust, True Float=$User.BurstIndexR, False Float=0)
    SpawnGroup=1
    SpawnRate=35
    Use Spawn Probability=true
   ) {module=SpawnRate001}
   &SpawnRate(
    Spawn Probability=&MakeCustomFloatFromBool(Bool=$User.IsSpawnDust, True Float=$User.BurstIndexMiddle, False Float=0)
    SpawnGroup=2
    SpawnRate=30
    Use Spawn Probability=true
   ) {module=SpawnRate002}
  ]
  particleSpawnScript [
   /SetVariables(Particles.VisibilityTag=&SpawnGroup()) {isSetParameters=true, module=SetVariables_#3}
   &InitializeParticle(
    Color=$System.SildingColor
    Lifetime Max=0.3
    Lifetime Min=0.07
    Lifetime Mode=::NewEnumerator1
    Mesh Scale Max=<1.5 0.2 1>
    Mesh Scale Min=<0.5 0.05 1>
    Mesh Scale Mode=::NewEnumerator2
    Sprite Size Mode=::NewEnumerator1
    Uniform Sprite Size Max=110
    Uniform Sprite Size Min=60
    UsePositionOffset=false
    Write Material Random=true
   )
   &SkeletalMeshLocation(
    Bone / Socket Index=&SpawnGroup()
    Bone Sampling Mode=::NewEnumerator4
    "Inherit Velocity (Scale)"=-1
    Interpolate Mesh Sampling=false
    MaskSpawnGroup=false
    Orientation Sampling=::NewEnumerator1
    Skeletal Mesh=$System.Skeletal Mesh
    Velocity Sampling=::NewEnumerator0
    VelocitySpeedLimit=false
   )
   &AddVelocity(
    Scale Added Velocity=&VectorFromFloat(Value=&FloatFromCurve(FloatCurve=DataInterface() {propertyValues {Curve {keys [<at 2.7256 0.4>,<at 3.2256 0.7>]}}}, CurveIndex=$Engine.System.Age))
    Velocity=&RandomRangeVector(Minimum=<0 -600 0>, Maximum=<0 -500 0>)
   )
   &AddVelocity(
    Velocity=&FN_FrontEndVector3Modifier(Gameplay=<0 0 -200>, Front End=<0 -600 -50>)
    Velocity Speed Scale=&MakeFloatFromVector(
     Vector=&VectorFromFloat(Value=&FloatFromCurve(FloatCurve=DataInterface() {propertyValues {Curve {keys [<at 3.4 1>,<at 3.5 0>]}}}, CurveIndex=$Engine.System.Age))
     Channel=::NewEnumerator0
    )
   ) {module=AddVelocity001}
   &ApplyInitialForces()
  ]
  particleUpdateScript [
   &ParticleState()
   &ScaleSpriteSize(
    Uniform Curve Scale=&MergeHelper_FloatSwitch(
     0=&RandomRangeFloat(Minimum=0.8, Maximum=1.2)
     1=&RandomRangeFloat(Minimum=0.8, Maximum=1.2)
     2=&RandomRangeFloat(Minimum=1.5, Maximum=2)
     3=?
     Index=&SpawnGroup()
     Num=::NewEnumerator1
    )
    Uniform Curve Sprite Scale=DataInterface() {propertyValues {Curve {keys [{arriveTangent=1.7988, interpMode=RCIM_Cubic, leaveTangent=1.7988, tangentMode=RCTM_User, time=0, value=0.2}, <at 1 1 RCIM_Cubic>]}}}
   )
   &ScaleColor(
    Scale Alpha=&FloatFromCurve(
     FloatCurve=DataInterface() {propertyValues {Curve {keys [
         {arriveTangent=-0.1787, interpMode=RCIM_Cubic, leaveTangent=-0.1787, tangentMode=RCTM_User, time=0, value=1}
         {arriveTangent=-1.7708, interpMode=RCIM_Cubic, leaveTangent=-1.7708, tangentMode=RCTM_User, time=1, value=0}
        ]}}}
     Scale Curve=&FN_FrontEndFloatModifier(
      Gameplay=&MergeHelper_FloatSwitch(
       0=0.55
       1=0.3
       2=?
       3=?
       Index=&MakeIntFromFloat(Float=$NPC.FortniteNPC.IsPlayerIndoors, Conversion Mode=::NewEnumerator1)
       Num=::NewEnumerator0
      )
      Front End=&Multiply_Float(
       A=1
       B=&FloatFromCurve(FloatCurve=DataInterface() {propertyValues {Curve {keys [<at 3.4 1.1>,<at 3.5 0.5>]}}}, CurveIndex=$Engine.System.Age, Scale Curve=0.85)
      )
     )
    )
    Scale RGB=&FN_FrontEndVector3Modifier(Gameplay=<1 1 1>, Front End=<2 2 2>)
    ScaleA=true
    ScaleRGB=true
   )
   &SolveForcesAndVelocity(Clamp Velocity=false, Limit Acceleration=false)
  ]
  rendererClasses [&NiagaraSpriteRendererProperties]
  renderers [
   {rendererClass=&NiagaraSpriteRendererProperties}
   {rendererClass=&NiagaraSpriteRendererProperties, rendererIndex=1}
   {rendererClass=&NiagaraSpriteRendererProperties, rendererIndex=2}
  ]
 }
 Foot_Spikes001 {
  emitterUpdateScript [
   &EmitterState()
   &SpawnRate(
    Spawn Probability=&MakeCustomFloatFromBool(Bool=$User.IsSpawnDust, True Float=$User.BurstIndexL, False Float=0)
    SpawnRate=&RandomRangeFloat(Minimum=9, Maximum=16, Evaluation Type=::NewEnumerator1)
    Use Spawn Probability=true
   )
   &SpawnRate(
    Spawn Probability=&MakeCustomFloatFromBool(Bool=$User.IsSpawnDust, True Float=$User.BurstIndexR, False Float=0)
    SpawnGroup=1
    SpawnRate=&RandomRangeFloat(Minimum=9, Maximum=16, Evaluation Type=::NewEnumerator1)
    Use Spawn Probability=true
   ) {module=SpawnRate001}
  ]
  particleSpawnScript [
   /SetVariables(Particles.VisibilityTag=&SpawnGroup()) {isSetParameters=true, module=SetVariables_#3}
   &InitializeParticle(
    Color=$System.SildingColor
    Lifetime Max=0.08
    Lifetime Min=0.04
    Lifetime Mode=::NewEnumerator1
    Mesh Scale Max=<2 0.25 1>
    Mesh Scale Min=<0.5 0.05 1>
    Mesh Scale Mode=::NewEnumerator2
    UsePositionOffset=false
    Write Material Random=true
   )
   &InitialMeshOrientation(Mesh Orientation Mode=::NewEnumerator2, Orientation Vector=<0 1 0>, Rotation=<0 0.47 0>, Use Rotation Vector=true)
   &AddVelocity(Velocity=<-50 0 0>)
   &ApplyInitialForces()
  ]
  particleUpdateScript [
   &ParticleState()
   &SkeletalMeshLocation(
    Bone / Socket Index=&SpawnGroup()
    Bone Sampling Mode=::NewEnumerator4
    MaskSpawnGroup=false
    Orientation Sampling=::NewEnumerator1
    Skeletal Mesh=$System.Skeletal Mesh
   )
   &ScaleMeshSize(Scale Factor=&MakeVector(
     X=&FloatFromCurve(
      FloatCurve=DataInterface() {propertyValues {Curve {keys [{arriveTangent=2.7289, interpMode=RCIM_Cubic, leaveTangent=2.7289, tangentMode=RCTM_User, time=0, value=0}, <at 1 1>]}}}
      Scale Curve=&RandomRangeFloat(Minimum=0.8, Maximum=2)
     )
     Y=&FloatFromCurve(
      FloatCurve=DataInterface() {propertyValues {Curve {keys [{arriveTangent=2.7289, interpMode=RCIM_Cubic, leaveTangent=2.7289, tangentMode=RCTM_User, time=0, value=0}, <at 1 1>]}}}
      Scale Curve=&RandomRangeFloat(Minimum=0.8, Maximum=1.5)
     )
     Z=1
    ))
   &ScaleMeshSize(Scale Factor=&FN_FrontEndVector3Modifier(Gameplay=<1 1 1>, Front End=<1 1 1>)) {module=ScaleMeshSize001}
   &ScaleColor(
    Scale Alpha=&FloatFromCurve(
     FloatCurve=DataInterface() {propertyValues {Curve {keys [
         {arriveTangent=-0.7355, interpMode=RCIM_Cubic, leaveTangent=-0.7355, tangentMode=RCTM_User, time=0.5, value=1}
         {arriveTangent=-3.1363, interpMode=RCIM_Cubic, leaveTangent=-3.1363, tangentMode=RCTM_User, time=1, value=0}
        ]}}}
     Scale Curve=&FloatFromCurve(
      FloatCurve=DataInterface() {propertyValues {Curve {keys [<at 7 0>,<at 7.3 1>]}}}
      CurveIndex=$Engine.System.Age
      Scale Curve=&FN_FrontEndFloatModifier(
       Gameplay=&MergeHelper_FloatSwitch(
        0=0.52
        1=0.3
        2=?
        3=?
        Index=&MakeIntFromFloat(Float=$NPC.FortniteNPC.IsPlayerIndoors, Conversion Mode=::NewEnumerator1)
        Num=::NewEnumerator0
       )
       Front End=0.4
      )
     )
    )
    Scale RGB=&FN_FrontEndVector3Modifier(
     Gameplay=&MakeVectorFromLinearColorRGB(Color=&ColorFromCurve(
       ColorCurve=DataInterface() {propertyValues {
         AlphaCurve {keys [<at 0 1 RCIM_Cubic>]}
         BlueCurve {keys [
           <at 0.0513 1.25>
           <at 0.1437 0.575>
           {arriveTangent=0.7845, interpMode=RCIM_Cubic, leaveTangent=0.7845, time=0.3527, value=0.9}
           {arriveTangent=0.0371, interpMode=RCIM_Cubic, leaveTangent=0.0371, time=0.558, value=0.9}
           {arriveTangent=0.9409, interpMode=RCIM_Cubic, leaveTangent=0.9409, time=0.8924, value=0.92}
           <at 0.93 1.25>
          ]}
         GreenCurve {keys [
           <at 0.0513 0.8125>
           <at 0.1437 0.6229>
           {arriveTangent=0.6688, interpMode=RCIM_Cubic, leaveTangent=0.6688, time=0.3527, value=0.9}
           {arriveTangent=0.0726, interpMode=RCIM_Cubic, leaveTangent=0.0726, time=0.558, value=0.9}
           {arriveTangent=-0.2352, interpMode=RCIM_Cubic, leaveTangent=-0.2352, time=0.8924, value=0.9392}
           <at 0.93 0.8125>
          ]}
         RedCurve {keys [
           <at 0.0513 0.5>
           <at 0.1437 1.15>
           {arriveTangent=-0.6034, interpMode=RCIM_Cubic, leaveTangent=-0.6034, time=0.3527, value=0.9}
           {arriveTangent=0.4632, interpMode=RCIM_Cubic, leaveTangent=0.4632, time=0.558, value=0.9}
           {arriveTangent=-1.0753, interpMode=RCIM_Cubic, leaveTangent=-1.0753, time=0.8924, value=1.15}
           <at 0.93 0.5>
          ]}
        }}
       CurveIndex=Linked() {linkedVariable {name=NPC.FortniteNPC.RemappedTimeOfDay_0-1, type=NiagaraFloat}}
      ))
     Front End=<1.5 1.5 1.5>
    )
    ScaleA=true
    ScaleRGB=true
   )
  ]
  rendererClasses [&NiagaraMeshRendererProperties]
  renderers [
   {rendererClass=&NiagaraMeshRendererProperties, rendererIndex=0}
   {rendererClass=&NiagaraMeshRendererProperties, rendererIndex=1}
  ]
 }
}
summary {
 systemName=NS_Sliding_Dust_ChiliAbide_LW
 emitters [
  {emitterName=Foot_Spikes001, rendererClasses [&NiagaraMeshRendererProperties]}
  {emitterName=Dust001, rendererClasses [&NiagaraSpriteRendererProperties]}
 ]
 userVariables [
  User.bIsFrontEnd {type=NiagaraBool}
  User.bIsFrontEndPreview {type=NiagaraBool}
  User.BurstIndexL {type=NiagaraFloat, defaultValue=1, description=""}
  User.BurstIndexMiddle {type=NiagaraFloat, defaultValue=0, description=""}
  User.BurstIndexR {type=NiagaraFloat, defaultValue=1, description=""}
  User.IsSpawnDust {type=NiagaraBool, defaultValue=true}
  User.SlidingColorArray {
   type=NiagaraDataInterfaceArrayColor
   description=""
   defaultValue {
    struct=NiagaraExt_VariableValue_DataInterface
    value {
     dataInterface=&self:NiagaraDataInterfaceArrayColor_0
     dataInterfaceClass=&NiagaraDataInterfaceArrayColor
    }
   }
  }
  User.SurfaceTypeIndex {type=NiagaraInt32}
 ]
}
